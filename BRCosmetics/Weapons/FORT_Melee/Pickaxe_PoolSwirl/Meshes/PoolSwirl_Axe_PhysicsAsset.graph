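# Physics asset for the PoolSwirl pickaxe mesh, a Battle Royale cosmetic melee weapon, giving it a simple collision shape.
# A single body on the b_AR_Root bone holds one capsule, offset and rotated along the axe.
# It has no constraints, so the asset defines no jointed or multi-body simulation.
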PhysicsAsset /BRCosmetics/Weapons/FORT_Melee/Pickaxe_PoolSwirl/Meshes/PoolSwirl_Axe_PhysicsAsset
bodies {b_AR_Root [{shapeType=Capsule, length=49.3142, radius=10.5267, center=<-2.7622 -0.0158 27.3093>, rotation=<rot -0.0577 -90 4.7472>}]}
constraints []
